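# Sound cue for searching a cooler loot container, played as player foley with prop-search attenuation.
# A quality-level node selects between two layered mixes of decreasing variety and two branches using one optimized cooler wave.
# Each mix blends quiet random cooler clips under louder random gear-search clips.
# Slight pitch and volume modulation varies the cooler sounds, and a concurrency override prevents new instances at the limit.
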SoundCue /Figment_S09/Sounds/Foley_Loot/Containers/Cooler/Cooler_Search_Cue
root=QualityLevel_0
nodes [
 {id=Mixer_0, children [Modulator_0,Random_1], properties {inputVolume [0.2,1.2]}}
 {id=Mixer_20, children [Modulator_23,Random_19], properties {inputVolume [0.2,1.2]}}
 {id=Modulator_0, children [Random_0], properties {pitchMax=1.07, pitchMin=0.92, volumeMax=1.07, volumeMin=0.92}}
 {id=Modulator_1, children [WavePlayer_6], properties {pitchMax=1.07, pitchMin=0.92, volumeMax=1.07, volumeMin=0.92}}
 {id=Modulator_2, children [WavePlayer_7], properties {pitchMax=1.07, pitchMin=0.92, volumeMax=1.07, volumeMin=0.92}}
 {id=Modulator_23, children [Random_45], properties {pitchMax=1.07, pitchMin=0.92, volumeMax=1.07, volumeMin=0.92}}
 {id=QualityLevel_0, children [Mixer_20,Mixer_0,Modulator_1,Modulator_2]}
 {id=Random_0, children [WavePlayer_0,WavePlayer_1], properties {weights [1,1]}}
 {id=Random_1, children [WavePlayer_3,WavePlayer_4], properties {weights [1,1]}}
 {id=Random_19, children [WavePlayer_99, WavePlayer_100, WavePlayer_102], properties {weights [1,1,1]}}
 {id=Random_45, children [WavePlayer_282, WavePlayer_283, WavePlayer_285], properties {weights [1,1,1]}}
 {id=WavePlayer_0, properties {soundWaveAssetPtr=/Game/Sounds/Foley_Loot/Containers/Cooler/Cooler_Search08}}
 {id=WavePlayer_1, properties {soundWaveAssetPtr=/Game/Sounds/Foley_Loot/Containers/Cooler/Cooler_Search01}}
 {id=WavePlayer_100, properties {soundWaveAssetPtr=/Game/Sounds/Foley_Loot/Containers/Misc/Prop_Gear_Search_06}}
 {id=WavePlayer_102, properties {soundWaveAssetPtr=/Game/Sounds/Foley_Loot/Containers/Misc/Prop_Gear_Search_01}}
 {id=WavePlayer_282, properties {soundWaveAssetPtr=&Cooler_Search08}}
 {id=WavePlayer_283, properties {soundWaveAssetPtr=&Cooler_Search01}}
 {id=WavePlayer_285, properties {soundWaveAssetPtr=/Game/Sounds/Foley_Loot/Containers/Cooler/Cooler_Search04}}
 {id=WavePlayer_3, properties {soundWaveAssetPtr=/Game/Sounds/Foley_Loot/Containers/Misc/Prop_Gear_Search_05}}
 {id=WavePlayer_4, properties {soundWaveAssetPtr=&Prop_Gear_Search_06}}
 {id=WavePlayer_6, properties {soundWaveAssetPtr=/Game/Sounds/Foley_Loot/Containers/Cooler/Cooler_Search_opt}}
 {id=WavePlayer_7, properties {soundWaveAssetPtr=&Cooler_Search_opt}}
 {id=WavePlayer_99, properties {soundWaveAssetPtr=&Prop_Gear_Search_05}}
]
properties {
 attenuationSettings=/Game/Sounds/Attenuation/Fort_Prop_Search_Att
 bOverrideConcurrency=true
 maxDistance=1300
 soundClassObject=&Fort_Player_Foley
 volumeMultiplier=2.4
 concurrencyOverrides {resolutionRule=PreventNew}
}
